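# Post-process style Animation Blueprint for the M_MED_MrMite player body cosmetic, adding secondary motion to collar, jacket and credentials.
# It copies the attached parent's pose through a Control Rig, then chains bone constraints, spring AnimDynamics on the collar, additive bone offsets and a RigidBody simulation.
# The update event computes skydive-reduced gravity and clamped thigh-to-spine and thigh-to-pelvis yaw factors, though the anim graph shows no reads of them.
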AnimBlueprint /BRCosmetics/Characters/Player/Male/Medium/Bodies/M_MED_MrMite/Meshes/M_MED_MrMite_AnimBP
target_skeleton=&/M_MED_MrMite_Skeleton
graphs [
 AnimGraph {nodes [
   n1 = Control Rig(Source=Copy Pose From Mesh() {bUseAttachedParent=true}) {
    alphaBoolBlend {blendOption=Linear}
    controlRigAssetReference {blueprintRigClass=&/M_MED_MrMite_ControlRig.M_MED_MrMite_ControlRig_C}
   }
   n2 = "Constraint - dyn_collar_CST_L (PPPPPPPP)"!(
    ComponentPose=Local To Component(LocalPose=@n1)
    Alpha=1
    ConstraintWeights_0=0.1359
    ConstraintWeights_1:float=0.098
    ConstraintWeights_2:float=0.0514
    ConstraintWeights_3:float=0.2139
    ConstraintWeights_4:float=0.0141
    ConstraintWeights_5:float=0.3113
    ConstraintWeights_6:float=0.0001
    ConstraintWeights_7:float=0.1752
   ) -> (Pose) {
    alphaBoolBlend {blendOption=Linear}
    boneToModify {boneName=dyn_collar_CST_L}
    constraintSetup [
     {targetBone {boneName=spine_03}}
     {targetBone {boneName=spine_04}}
     {targetBone {boneName=spine_05}}
     {targetBone {boneName=clavicle_l}}
     {targetBone {boneName=upperarm_twist_01_l}}
     {targetBone {boneName=dyn_collar_01_L}}
     {targetBone {boneName=dyn_collar_02_L}}
     {targetBone {boneName=dyn_jacket_fr_01_L}}
    ]
    constraintWeights [1,1,1,1,1,1,1,1]
   }
   n3 = "Constraint - dyn_collar_CST_R (PPPPPPPP)"!(
    ComponentPose=@n2
    Alpha=1
    ConstraintWeights_0=0.1255
    ConstraintWeights_1:float=0.09
    ConstraintWeights_2:float=0.0491
    ConstraintWeights_3:float=0.2101
    ConstraintWeights_4:float=0.015
    ConstraintWeights_5:float=0.3685
    ConstraintWeights_6:float=0.0004
    ConstraintWeights_7:float=0.1414
   ) -> (Pose) {
    alphaBoolBlend {blendOption=Linear}
    boneToModify {boneName=dyn_collar_CST_R}
    constraintSetup [
     {targetBone {boneName=spine_03}}
     {targetBone {boneName=spine_04}}
     {targetBone {boneName=spine_05}}
     {targetBone {boneName=clavicle_r}}
     {targetBone {boneName=upperarm_twist_01_r}}
     {targetBone {boneName=dyn_collar_01_R}}
     {targetBone {boneName=dyn_collar_02_R}}
     {targetBone {boneName=dyn_jacket_fr_01_R}}
    ]
    constraintWeights [1,1,1,1,1,1,1,1]
   }
   n4 = "Constraint - dyn_jacket_fr_CST_L (P)"(ComponentPose=@n3, Alpha=1, ConstraintWeights_0=1) {
    alphaBoolBlend {blendOption=Linear}
    boneToModify {boneName=dyn_jacket_fr_CST_L}
    constraintSetup [{targetBone {boneName=spine_01}}]
    constraintWeights [1]
   }
   n5 = "Constraint - dyn_jacket_CST_L (P)"(ComponentPose=@n4, Alpha=1, ConstraintWeights_0=1) {
    alphaBoolBlend {blendOption=Linear}
    boneToModify {boneName=dyn_jacket_CST_L}
    constraintSetup [{targetBone {boneName=spine_01}}]
    constraintWeights [1]
   }
   n6 = "Constraint - credentials_CST (PP)"!(ComponentPose=@n5, Alpha=1, ConstraintWeights_0=0.8371, ConstraintWeights_1:float=0.1629) -> (Pose) {
    alphaBoolBlend {blendOption=Linear}
    boneToModify {boneName=credentials_CST}
    constraintSetup [{targetBone {boneName=pelvis}}, {targetBone {boneName=spine_01}}]
    constraintWeights [1,1]
   }
   n7 = Anim Dynamics - Bone: dyn_collar_01_L!(ComponentPose=@n6, Alpha=1, GravityOverride:Vector="0,0,-980") -> (Pose) {
    angularBiasOverride=0.999
    angularDampingOverride=0.99
    angularSpringConstant=1000
    bAngularSpring=true
    bOverrideAngularDamping=true
    bUseGravityOverride=true
    simulationSpace=RootRelative
    alphaBoolBlend {blendOption=Linear}
    boundBone {boneName=dyn_collar_01_L}
    componentAppliedLinearAccClamp=<5000 5000 5000>
    componentLinearAccScale=<0.5 0.5 0.5>
    componentLinearVelScale=<0.5 0.5 0.5>
    physicsBodyDefinitions [{
      boundBone {boneName=dyn_collar_01_L}
      boxExtents=<15 15 15>
      constraintSetup {angularLimitsMax=<0 15 0>, angularLimitsMin=<0 -5 0>, angularTarget=<1 0 0>}
      localJointOffset=<0 0 0>
     }]
   }
   n8 = Anim Dynamics - Bone: dyn_collar_02_L!(ComponentPose=@n7, Alpha=1, GravityOverride:Vector="0,0,-980") -> (Pose) {
    angularBiasOverride=0.999
    angularDampingOverride=0.99
    angularSpringConstant=1000
    bAngularSpring=true
    bOverrideAngularDamping=true
    bUseGravityOverride=true
    bUseSphericalLimits=true
    simulationSpace=RootRelative
    alphaBoolBlend {blendOption=Linear}
    boundBone {boneName=dyn_collar_02_L}
    componentAppliedLinearAccClamp=<5000 5000 5000>
    componentLinearAccScale=<0.5 0.5 0.5>
    componentLinearVelScale=<0.5 0.5 0.5>
    physicsBodyDefinitions [{
      collisionType=CustomSphere
      sphereCollisionRadius=1
      boundBone {boneName=dyn_collar_02_L}
      boxExtents=<7 7 7>
      constraintSetup {angularLimitsMax=<0 15 0>, angularLimitsMin=<0 -5 0>, angularTarget=<1 0 0>}
      localJointOffset=<0 0 0>
     }]
    sphericalLimits [{limitRadius=8, drivingBone {boneName=clavicle_l}, sphereLocalOffset=<8.5 0 -2.1277>}]
   }
   n9 = Anim Dynamics - Bone: dyn_collar_01_R!(ComponentPose=@n8, Alpha=1, GravityOverride:Vector="0,0,-980") -> (Pose) {
    angularBiasOverride=0.999
    angularDampingOverride=0.99
    angularSpringConstant=1000
    bAngularSpring=true
    bOverrideAngularDamping=true
    bUseGravityOverride=true
    simulationSpace=RootRelative
    alphaBoolBlend {blendOption=Linear}
    boundBone {boneName=dyn_collar_01_R}
    componentAppliedLinearAccClamp=<5000 5000 5000>
    componentLinearAccScale=<0.5 0.5 0.5>
    componentLinearVelScale=<0.5 0.5 0.5>
    physicsBodyDefinitions [{
      boundBone {boneName=dyn_collar_01_R}
      boxExtents=<15 15 15>
      constraintSetup {angularLimitsMax=<0 15 0>, angularLimitsMin=<0 -5 0>, angularTarget=<1 0 0>}
      localJointOffset=<0 0 0>
     }]
   }
   n10 = Anim Dynamics - Bone: dyn_collar_02_R!(ComponentPose=@n9, Alpha=1, GravityOverride:Vector="0,0,-980") -> (Pose) {
    angularBiasOverride=0.999
    angularDampingOverride=0.99
    angularSpringConstant=1000
    bAngularSpring=true
    bOverrideAngularDamping=true
    bUseGravityOverride=true
    bUseSphericalLimits=true
    simulationSpace=RootRelative
    alphaBoolBlend {blendOption=Linear}
    boundBone {boneName=dyn_collar_02_R}
    componentAppliedLinearAccClamp=<5000 5000 5000>
    componentLinearAccScale=<0.5 0.5 0.5>
    componentLinearVelScale=<0.5 0.5 0.5>
    physicsBodyDefinitions [{
      collisionType=CustomSphere
      sphereCollisionRadius=1
      boundBone {boneName=dyn_collar_02_R}
      boxExtents=<7 7 7>
      constraintSetup {angularLimitsMax=<0 15 0>, angularLimitsMin=<0 -5 0>, angularTarget=<1 0 0>}
      localJointOffset=<0 0 0>
     }]
    sphericalLimits [{limitRadius=8, drivingBone {boneName=clavicle_r}, sphereLocalOffset=<-8.5 0 -2.1277>}]
   }
   n11 = "Transform (Modify) Bone - Bone: credentials"!(ComponentPose=@n10, Alpha=1, Translation="4,0,2.5", Rotation="30,2,10") -> (Pose) {
    rotationMode=BMM_Additive
    rotationSpace=BCS_BoneSpace
    translationMode=BMM_Additive
    translationSpace=BCS_BoneSpace
    alphaBoolBlend {blendOption=Linear}
    boneToModify {boneName=credentials}
    rotation=<rot 30 2 10>
    translation {x=4, z=2.5}
   }
   n12 = "Transform (Modify) Bone - Bone: dyn_jacket_fr_MSTR_L"!(ComponentPose=@n11, Alpha=1, Translation="0,3,10", Rotation="0,-80,0") -> (Pose) {
    rotationMode=BMM_Additive
    rotationSpace=BCS_BoneSpace
    translationMode=BMM_Additive
    translationSpace=BCS_BoneSpace
    alphaBoolBlend {blendOption=Linear}
    alphaScaleBiasClamp {interpSpeedDecreasing=5, interpSpeedIncreasing=5}
    boneToModify {boneName=dyn_jacket_fr_MSTR_L}
    rotation {yaw=-80}
    translation {y=3, z=10}
   }
   n13 = RigidBody!(ComponentPose=@n12, Alpha, OverrideWorldGravity:Vector="0,0,-980") -> (Pose) {
    bClampLinearTranslationLimitToRefPose=true
    bOverrideWorldGravity=true
    simulationSpace=BaseBoneSpace
    alphaBoolBlend {blendOption=Linear}
    baseBoneRef {boneName=spine_03}
    simSpaceSettings {
     dampingAlpha=0
     maxAngularAcceleration=6
     maxAngularVelocity=5.5
     maxLinearAcceleration=500
     maxLinearVelocity=1000
     worldAlpha=1
     externalLinearDragV=<1 1 1>
    }
   }
   Output Pose(Result=Component To Local(ComponentPose=@n13)) {layerGroup=None}
  ]}
 EventGraph {nodes [
   n1 = Event BlueprintUpdateAnimation()
   > Set Delta Time X(Delta Time X=@n1.DeltaTimeX)
   > n2 = Sequence()
   n3 = Get bIsSkydiveFloating()
   n4 = Get bIsSkydiveDiveMode()
   > Set gravity overrride(gravity overrride:Vector=SelectVector(A="0,0,-100", B="0,0,-980", bPickA=OR Boolean(A=@n3, B=@n4))) -> (Output_Get:Vector)
   n5 = Get bIsMoving2D()
   n6 = Get thigh l/Spine yaw() -> (thigh l/Spine yaw:double)
   n7 = Get Delta Time X()
   n8 = GetOwnerSkeletalMeshComponent()
   n9 = GetDeltaTransformFromRefPose(self=@n8, BoneName=spine_05, BaseName=thigh_l)
   n10 = MapRangeClamped(Value=BreakRotator(InRot=BreakTransform(InTransform=@n9).Rotation).Yaw, InRangeA=100, InRangeB=170)
   n11 = Get thigh l/Spine yaw()
   n12 = Get Delta Time X()
   Set thigh l/Spine yaw(
    execute=@n2.then_1
    thigh l/Spine yaw:double=SelectFloat(
     A=FInterpTo(Current=@n11, Target=@n10, DeltaTime=@n12, InterpSpeed=5)
     B=FInterpTo(Current=@n6, Target=0, DeltaTime=@n7, InterpSpeed=5)
     bPickA=NOT Boolean(A=@n5)
    )
   ) -> (Output_Get:double)
   n13 = BreakTransform(InTransform=GetDeltaTransformFromRefPose(self=@n8, BoneName=thigh_l, BaseName=pelvis))
   > Set thigh l/Pelvis yaw(thigh l/Pelvis yaw:double=MapRangeClamped(Value=BreakRotator(InRot=@n13.Rotation).Yaw, InRangeA=-50, InRangeB=-90)) -> (Output_Get:double)
  ]}
]
variables [
 Delta Time X {type=number}
 gravity overrride {type=Vector}
 thigh l/Pelvis yaw {type=number}
 thigh l/Spine yaw {type=number}
]
