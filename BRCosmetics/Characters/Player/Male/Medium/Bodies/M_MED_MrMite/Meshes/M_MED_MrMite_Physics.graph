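# Physics asset for the MrMite medium male player body, set up for a dynamic jacket.
# Capsules cover pelvis, spine, thighs and the right thigh twist, plus seven front, side and back jacket chains.
# Each chain hangs from spine_03 through a tiny master body with free angular motion, then two segments with narrow angular limits.
# Angularly free cross-constraints tie side tips to front and outer back tips, leaving the center back unlinked.
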PhysicsAsset /BRCosmetics/Characters/Player/Male/Medium/Bodies/M_MED_MrMite/Meshes/M_MED_MrMite_Physics
bodies {
 dyn_jacket_01_L [{shapeType=Capsule, length=3.5, radius=7, shapeName=dyn_jacket_01_L_capsule, center=<5 0 3>, rotation=<rot 0 -90 90>}]
 dyn_jacket_01_R [{shapeType=Capsule, length=3.5, radius=7, shapeName=dyn_jacket_01_R_capsule, center=<3 0 3>, rotation=<rot 0 -90 90>}]
 dyn_jacket_02_L [{shapeType=Capsule, length=3.5, radius=7, shapeName=dyn_jacket_02_L_capsule, center=<5 0 3>, rotation=<rot 0 -90 90>}]
 dyn_jacket_02_R [{shapeType=Capsule, length=3.5, radius=7, shapeName=dyn_jacket_02_R_capsule, center=<3 0 3>, rotation=<rot 0 -90 90>}]
 dyn_jacket_MSTR_L [{shapeType=Capsule, length=0, radius=1, shapeName=dyn_jacket_MSTR_L_capsule, rotation=<rot 0 -89.9983 90.0002>}]
 dyn_jacket_MSTR_R [{shapeType=Capsule, length=0, radius=1, shapeName=dyn_jacket_MSTR_R_capsule, rotation=<rot 0 -90.0017 90.0002>}]
 dyn_jacket_bk_01 [{shapeType=Capsule, length=3.5, radius=7, shapeName=dyn_jacket_bk_01_capsule, center=<7 0 3>, rotation=<rot 0 -90 90>}]
 dyn_jacket_bk_01_L [{shapeType=Capsule, length=3.5, radius=7, shapeName=dyn_jacket_bk_01_L_capsule, center=<7 0 3>, rotation=<rot 0 -90 90>}]
 dyn_jacket_bk_01_R [{shapeType=Capsule, length=3.5, radius=7, shapeName=dyn_jacket_bk_01_R_capsule, center=<7 0 3>, rotation=<rot 0 -90 90>}]
 dyn_jacket_bk_02 [{shapeType=Capsule, length=3.5, radius=7, shapeName=dyn_jacket_bk_02_capsule, center=<7 0 3>, rotation=<rot 0 -90 90>}]
 dyn_jacket_bk_02_L [{shapeType=Capsule, length=3.5, radius=7, shapeName=dyn_jacket_bk_02_L_capsule, center=<7 0 3>, rotation=<rot 0 -90 90>}]
 dyn_jacket_bk_02_R [{shapeType=Capsule, length=3.5, radius=7, shapeName=dyn_jacket_bk_02_R_capsule, center=<7 0 3>, rotation=<rot 0 -90 90>}]
 dyn_jacket_bk_MSTR [{shapeType=Capsule, length=0, radius=1, shapeName=dyn_jacket_bk_MSTR_capsule, rotation=<rot 0 -90 90.0017>}]
 dyn_jacket_bk_MSTR_L [{shapeType=Capsule, length=0, radius=1, shapeName=dyn_jacket_bk_MSTR_L_capsule, rotation=<rot 0 -89.999 90.0014>}]
 dyn_jacket_bk_MSTR_R [{shapeType=Capsule, length=0, radius=1, shapeName=dyn_jacket_bk_MSTR_R_capsule, rotation=<rot 0 -90.001 90.0014>}]
 dyn_jacket_fr_01_L [{shapeType=Capsule, length=3.5, radius=7.5, shapeName=dyn_jacket_fr_01_L_capsule, center=<7 0 3>, rotation=<rot 0 -90 90>}]
 dyn_jacket_fr_01_R [{shapeType=Capsule, length=3.5, radius=7.5, shapeName=dyn_jacket_fr_01_R_capsule, center=<7 0 3>, rotation=<rot 0 -90 90>}]
 dyn_jacket_fr_02_L [{shapeType=Capsule, length=2.5, radius=7, shapeName=dyn_jacket_fr_02_L_capsule, center=<7 0 3>, rotation=<rot 0 -90 90>}]
 dyn_jacket_fr_02_R [{shapeType=Capsule, length=3.5, radius=7, shapeName=dyn_jacket_fr_02_R_capsule, center=<7 0 3>, rotation=<rot 0 -90 90>}]
 dyn_jacket_fr_MSTR_L [{shapeType=Capsule, length=0, radius=1, shapeName=dyn_jacket_fr_MSTR_L_capsule, rotation=<rot -0.0001 -89.9988 89.9989>}]
 dyn_jacket_fr_MSTR_R [{shapeType=Capsule, length=0, radius=1, shapeName=dyn_jacket_fr_MSTR_R_capsule, rotation=<rot 0.0001 -90.0011 89.9989>}]
 pelvis [{shapeType=Capsule, length=9.5, radius=10.5, shapeName=pelvis_capsule, center=<-2.2627 -1.232 0>}]
 spine_01 [{shapeType=Capsule, length=4, radius=10, shapeName=spine_01_capsule}]
 spine_02 [{shapeType=Capsule, length=5, radius=8.5, shapeName=spine_02_capsule, center=<3.1218 0 0>}]
 spine_03 [{shapeType=Capsule, length=8.7, radius=7.5, shapeName=spine_03_capsule, center=<5.6155 0 0>}]
 thigh_l [{shapeType=Capsule, length=34, radius=8, shapeName=thigh_l_capsule, center=<-18 0 0>, rotation=<rot -90 0 0>}]
 thigh_r [{shapeType=Capsule, length=34, radius=8, shapeName=thigh_r_capsule, center=<18 0 0>, rotation=<rot -90 0 0>}]
 thigh_twist_01_r [{shapeType=Capsule, length=7.5, radius=9, shapeName=thigh_twist_01_r_capsule, center=<-13.8806 -1.2974 6.4137>, rotation=<rot -90 -0.0001 0>}]
}
constraints [
 {
  bone1Name=spine_01
  bone2Name=pelvis
  swing1LimitDegrees=45
  swing1Motion=Free
  swing2LimitDegrees=45
  swing2Motion=Free
  twistLimitDegrees=45
  twistMotion=Free
 }
 {
  bone1Name=spine_02
  bone2Name=spine_01
  swing1LimitDegrees=45
  swing1Motion=Free
  swing2LimitDegrees=45
  swing2Motion=Free
  twistLimitDegrees=45
  twistMotion=Free
 }
 {
  bone1Name=spine_03
  bone2Name=spine_02
  swing1LimitDegrees=45
  swing1Motion=Free
  swing2LimitDegrees=45
  swing2Motion=Free
  twistLimitDegrees=45
  twistMotion=Free
 }
 {
  bone1Name=dyn_jacket_MSTR_L
  bone2Name=spine_03
  swing1LimitDegrees=45
  swing1Motion=Free
  swing2LimitDegrees=45
  swing2Motion=Free
  twistLimitDegrees=45
  twistMotion=Free
 }
 {bone1Name=dyn_jacket_01_L, bone2Name=dyn_jacket_MSTR_L, swing1LimitDegrees=5, swing2LimitDegrees=15, twistLimitDegrees=5}
 {bone1Name=dyn_jacket_02_L, bone2Name=dyn_jacket_01_L, swing1LimitDegrees=10, swing2LimitDegrees=25, twistLimitDegrees=5}
 {
  bone1Name=dyn_jacket_MSTR_R
  bone2Name=spine_03
  swing1LimitDegrees=45
  swing1Motion=Free
  swing2LimitDegrees=45
  swing2Motion=Free
  twistLimitDegrees=45
  twistMotion=Free
 }
 {bone1Name=dyn_jacket_01_R, bone2Name=dyn_jacket_MSTR_R, swing1LimitDegrees=5, swing2LimitDegrees=15, twistLimitDegrees=5}
 {bone1Name=dyn_jacket_02_R, bone2Name=dyn_jacket_01_R, swing1LimitDegrees=10, swing2LimitDegrees=25, twistLimitDegrees=5}
 {
  bone1Name=dyn_jacket_bk_MSTR
  bone2Name=spine_03
  swing1LimitDegrees=45
  swing1Motion=Free
  swing2LimitDegrees=45
  swing2Motion=Free
  twistLimitDegrees=45
  twistMotion=Free
 }
 {bone1Name=dyn_jacket_bk_01, bone2Name=dyn_jacket_bk_MSTR, swing1LimitDegrees=5, swing2LimitDegrees=15, twistLimitDegrees=5}
 {bone1Name=dyn_jacket_bk_02, bone2Name=dyn_jacket_bk_01, swing1LimitDegrees=10, swing2LimitDegrees=25, twistLimitDegrees=5}
 {
  bone1Name=dyn_jacket_bk_MSTR_L
  bone2Name=spine_03
  swing1LimitDegrees=45
  swing1Motion=Free
  swing2LimitDegrees=45
  swing2Motion=Free
  twistLimitDegrees=45
  twistMotion=Free
 }
 {bone1Name=dyn_jacket_bk_01_L, bone2Name=dyn_jacket_bk_MSTR_L, swing1LimitDegrees=5, swing2LimitDegrees=15, twistLimitDegrees=5}
 {bone1Name=dyn_jacket_bk_02_L, bone2Name=dyn_jacket_bk_01_L, swing1LimitDegrees=10, swing2LimitDegrees=25, twistLimitDegrees=5}
 {
  bone1Name=dyn_jacket_bk_MSTR_R
  bone2Name=spine_03
  swing1LimitDegrees=45
  swing1Motion=Free
  swing2LimitDegrees=45
  swing2Motion=Free
  twistLimitDegrees=45
  twistMotion=Free
 }
 {bone1Name=dyn_jacket_bk_01_R, bone2Name=dyn_jacket_bk_MSTR_R, swing1LimitDegrees=5, swing2LimitDegrees=15, twistLimitDegrees=5}
 {bone1Name=dyn_jacket_bk_02_R, bone2Name=dyn_jacket_bk_01_R, swing1LimitDegrees=10, swing2LimitDegrees=25, twistLimitDegrees=5}
 {
  bone1Name=dyn_jacket_fr_MSTR_L
  bone2Name=spine_03
  swing1LimitDegrees=45
  swing1Motion=Free
  swing2LimitDegrees=45
  swing2Motion=Free
  twistLimitDegrees=45
  twistMotion=Free
 }
 {bone1Name=dyn_jacket_fr_01_L, bone2Name=dyn_jacket_fr_MSTR_L, swing1LimitDegrees=5, swing2LimitDegrees=15, twistLimitDegrees=5}
 {bone1Name=dyn_jacket_fr_02_L, bone2Name=dyn_jacket_fr_01_L, swing1LimitDegrees=10, swing2LimitDegrees=25, twistLimitDegrees=5}
 {
  bone1Name=dyn_jacket_fr_MSTR_R
  bone2Name=spine_03
  swing1LimitDegrees=45
  swing1Motion=Free
  swing2LimitDegrees=45
  swing2Motion=Free
  twistLimitDegrees=45
  twistMotion=Free
 }
 {bone1Name=dyn_jacket_fr_01_R, bone2Name=dyn_jacket_fr_MSTR_R, swing1LimitDegrees=5, swing2LimitDegrees=15, twistLimitDegrees=5}
 {bone1Name=dyn_jacket_fr_02_R, bone2Name=dyn_jacket_fr_01_R, swing1LimitDegrees=10, swing2LimitDegrees=25, twistLimitDegrees=5}
 {
  bone1Name=thigh_l
  bone2Name=pelvis
  swing1LimitDegrees=45
  swing1Motion=Free
  swing2LimitDegrees=45
  swing2Motion=Free
  twistLimitDegrees=45
  twistMotion=Free
 }
 {
  bone1Name=thigh_r
  bone2Name=pelvis
  swing1LimitDegrees=45
  swing1Motion=Free
  swing2LimitDegrees=45
  swing2Motion=Free
  twistLimitDegrees=45
  twistMotion=Free
 }
 {
  bone1Name=dyn_jacket_02_R
  bone2Name=dyn_jacket_fr_02_R
  swing1LimitDegrees=45
  swing1Motion=Free
  swing2LimitDegrees=45
  swing2Motion=Free
  twistLimitDegrees=45
  twistMotion=Free
 }
 {
  bone1Name=dyn_jacket_02_L
  bone2Name=dyn_jacket_fr_02_L
  swing1LimitDegrees=45
  swing1Motion=Free
  swing2LimitDegrees=45
  swing2Motion=Free
  twistLimitDegrees=45
  twistMotion=Free
 }
 {
  bone1Name=dyn_jacket_bk_02_L
  bone2Name=dyn_jacket_02_L
  swing1LimitDegrees=45
  swing1Motion=Free
  swing2LimitDegrees=45
  swing2Motion=Free
  twistLimitDegrees=45
  twistMotion=Free
 }
 {
  bone1Name=dyn_jacket_bk_02_R
  bone2Name=dyn_jacket_02_R
  swing1LimitDegrees=45
  swing1Motion=Free
  swing2LimitDegrees=45
  swing2Motion=Free
  twistLimitDegrees=45
  twistMotion=Free
 }
 {
  bone1Name=thigh_twist_01_r
  bone2Name=thigh_r
  swing1LimitDegrees=45
  swing1Motion=Free
  swing2LimitDegrees=45
  swing2Motion=Free
  twistLimitDegrees=45
  twistMotion=Free
 }
]
